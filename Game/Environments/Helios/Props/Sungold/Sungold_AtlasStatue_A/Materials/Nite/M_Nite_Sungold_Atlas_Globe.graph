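# Nanite, lit material for the globe of the Nite Sungold Atlas statue, showing an animated glowing cosmos.
# Emissive layers foreground and background constellations, stars, clouds and blue lines, using drifting UVs, noise wobble and shimmer masks.
# Ray-traced-sphere normals and stabilized screen-space UVs build matcap coordinates and Fresnel masks for faked depth.
# Tinted, desaturated base color varies with sky light direction; a mask texture drives specular, roughness and metallic.
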Material /Game/Environments/Helios/Props/Sungold/Sungold_AtlasStatue_A/Materials/Nite/M_Nite_Sungold_Atlas_Globe
graph {
 properties {bUseMaterialAttributes=true, bUsedWithNanite=true, pixelDepthOffsetMode=PDOM_Legacy, usedShadingModels=Default Lit}
 nodes [
  n1 = ObjectPositionWS() {originType=CameraRelative}
  Abs(None=@n1)
  n2 = Constant3Vector() {constant=<rgba 1 -1 -1 0>}
  n3 = Multiply(A=@n2, B=Normalize()) {constB=-0.5}
  Add(A=@n3) {constB=0.5}
  n4 = Multiply(A=DotProduct(A=SkyAtmosphereLightDirection(), B=VertexNormalWS()), B=Constant() {r=0.5}) {constB=-0.5}
  Add(A=@n4, B=Constant()) {constB=0}
  n5 = NamedRerouteUsage()
  n6 = Add(A=@n5, B=Multiply(A=NamedRerouteUsage(), B=Constant() {r=-0.05}) {constB=-0.05, desc=-0.05})
  AppendVector(A=@n6, B=ComponentMask(None=@n5) {g=true})
  BumpOffset(
   Coordinate=MaterialFunctionCall!(Position, UVRadialScale) {materialFunction=/Game/Weapons/WeaponSkins/MaterialFunction/ObjectStabilizedScreenSpaceUVs}.StabilizedUVs
   Height=Constant() {r=5}
  )
  ComponentMask(None=@n3) {g=true, r=true}
  ComponentMask(None=@n5) {r=true}
  ComponentMask(None=@n1) {r=true}
  n7 = PixelNormalWS()
  n8 = NamedRerouteDeclaration(None=DotProduct(A=@n7, B=CameraVectorWS())) {name=Fresnel_1, nodeColor=<rgba 0.7578 1 0 1>}
  n9 = Add(A=@n8, B=Constant() {r=0.6}) {constB=0.6}
  n10 = Transform(None=Multiply(A=@n7, B=Saturate(None=@n9))) {transformSourceType=TRANSFORMSOURCE_World, transformType=TRANSFORM_View}
  ComponentMask(None=Add(A=Multiply(A=@n10) {constB=-0.5}, B=@n2) {constB=0.5}) {g=true, r=true}
  n11 = TextureSample(Tex=NamedRerouteUsage()) {samplerType=SAMPLERTYPE_Masks}
  n12 = Add(
   A=Multiply(A=@n11.R, B=ScalarParameter() {defaultValue=1, desc=0.3, group=Cosmos, parameterName=Specular Intensity, sortPriority=0})
   B=ScalarParameter() {desc=0, group=Cosmos, parameterName=Specular Add, sortPriority=0}
  )
  n13 = ScalarParameter() {defaultValue=0.8, desc=0.8, group=Cosmos, parameterName=Roughness Intensity, sortPriority=0}
  n14 = ScalarParameter() {defaultValue=0.1, desc=0.15, group=Cosmos, parameterName=Roughness Add, sortPriority=0}
  n15 = ScalarParameter() {defaultValue=2, desc=1, group=Cosmos, parameterName=Metallic Intensity, sortPriority=0}
  n16 = NamedRerouteUsage()
  n17 = Constant() {r=-0.2}
  n18 = Multiply(A=Constant() {r=15}, B=Saturate(None=Add(A=@n16, B=@n17)))
  n19 = NamedRerouteUsage()
  n20 = Multiply(A=@n19, B=@n19)
  n21 = ScalarParameter() {defaultValue=11.5, parameterName=Param_1}
  n22 = NamedRerouteUsage()
  n23 = NamedRerouteUsage()
  n24 = Constant() {r=0.05}
  n25 = AppendVector(A=Add(A=ComponentMask(None=@n22) {r=true}, B=Multiply(A=@n23, B=@n24) {constB=0.05, desc=0.05}), B=ComponentMask(None=@n22) {g=true})
  n26 = TextureSample(UVs=@n25, Tex=NamedRerouteUsage())
  n27 = NamedRerouteDeclaration(None=@n26.RGB) {name=ColorVariation, nodeColor=<rgba 0.4766 0 1 1>}
  n28 = Add(
   A=TextureSample(Tex=NamedRerouteUsage()).RGB
   B=Multiply(A=Multiply(A=@n27, B=@n26.A), B=VectorParameter() {parameterName=Globe Base Color Variation, defaultValue=<rgba 0.125 0.4046 1.2 1>}.RGB) {constB=2}
  )
  n29 = VectorParameter() {parameterName=Globe Base Color add, defaultValue=<rgba 0.0664 0.0757 0.25 1>}
  n30 = Multiply(
   A=Add(A=Multiply(A=@n20, B=@n21), B=ScalarParameter() {defaultValue=0.125}) {constB=0.1}
   B=Desaturation(None=Add(A=@n28, B=@n29.RGB), Fraction=ScalarParameter() {defaultValue=0.45, group=Cosmos, parameterName=Basecolor Desat})
  )
  n31 = Saturate(None=@n4)
  n32 = NamedRerouteUsage()
  n33 = Constant() {r=3}
  n34 = Saturate(None=Add(A=Multiply(A=@n32, B=@n33) {desc=1.5}, B=Constant() {r=-2}) {desc=-0.5})
  n35 = VectorParameter() {group=Cosmos, parameterName=Constelations FG Color, defaultValue=<rgba 6 3.8604 1.0074 1>}
  n36 = TextureCoordinate()
  n37 = Multiply(A=ScalarParameter() {defaultValue=0.1, desc=0.05, parameterName=Constellations FG Distortion Speed}, B=NamedRerouteUsage())
  n38 = TextureSample(UVs=Multiply(A=ScalarParameter() {defaultValue=1.1, parameterName=Constellations FG Distortion Tile}, B=@n36), Tex=NamedRerouteUsage())
  n39 = Saturate(None=Sine(None=Add(A=@n37, B=@n38.RGB)))
  n40 = TextureSample(
   UVs=LinearInterpolate(
    A=@n36
    B=ComponentMask(None=@n39) {g=true, r=true}
    Alpha=ScalarParameter() {defaultValue=0.003, desc=0.0006, parameterName=Constellation FG Distortion Intensity}
   )
   Tex=NamedRerouteUsage()
  )
  n41 = Multiply(A=@n40.R, B=ScalarParameter() {defaultValue=0.2, parameterName=sadfaggs})
  n42 = NamedRerouteUsage()
  n43 = NamedRerouteUsage()
  n44 = Multiply(A=@n43) {constB=1.5}
  n45 = Multiply(A=Saturate(None=Add(A=@n44) {constB=0.1}), B=@n40.B)
  n46 = Multiply(A=@n43, B=@n40.G)
  n47 = Multiply(A=NamedRerouteUsage(), B=Constant() {r=0.5}) {desc=0.9}
  n48 = Saturate(None=Add(A=@n47, B=Constant() {r=0.05}) {desc=0.08})
  n49 = Add(A=Multiply(A=NamedRerouteUsage()) {constB=3}) {constB=-0.1}
  n50 = NamedRerouteUsage()
  n51 = ComponentMask(None=@n50) {r=true}
  n52 = Multiply(A=NamedRerouteDeclaration(None=Time()) {name=Time, nodeColor=<rgba 0 1 0.1797 1>}, B=Constant() {r=-0.002}) {constB=-0.005}
  n53 = Add(A=@n51, B=@n52)
  n54 = ComponentMask(None=@n50) {g=true}
  n55 = TextureSample(UVs=Multiply(A=AppendVector(A=@n53, B=@n54), B=Constant() {r=3.2}) {constB=0.18}, Tex=NamedRerouteUsage())
  n56 = VectorParameter() {group=Cosmos, parameterName=Stars FG Color, defaultValue=<rgba 3.3355 5.089 15 1>}
  n57 = NamedRerouteUsage()
  n58 = Multiply(A=@n57) {constB=1.5}
  n59 = NamedRerouteUsage()
  n60 = ComponentMask(None=@n59) {r=true}
  n61 = NamedRerouteUsage()
  n62 = ComponentMask(None=@n59) {g=true}
  n63 = AppendVector(A=Add(A=@n60, B=Multiply(A=@n61) {constB=0.0015}), B=@n62)
  n64 = TextureSample(
   UVs=Multiply(A=@n63, B=ScalarParameter() {defaultValue=1.8, group=Cosmos, parameterName=Constelations BG Tiling}) {constB=2.5}
   Tex=NamedRerouteUsage()
  )
  n65 = Multiply(A=Saturate(None=Add(A=@n58) {constB=0.1}), B=@n64.B)
  n66 = Constant() {r=3}
  n67 = Multiply(A=Add(A=Multiply(A=@n57, B=@n66), B=Constant() {r=0.2}), B=@n64.G)
  n68 = Multiply(
   A=Add(A=Add(A=@n65, B=@n67), B=@n64.R)
   B=VectorParameter() {group=Cosmos, parameterName=Constelations BG Color, defaultValue=<rgba 2.7394 3.6603 8 1>}.RGB
  )
  n69 = Multiply(A=NamedRerouteUsage())
  n70 = Multiply(A=@n52) {constB=-0.5}
  n71 = Multiply(A=AppendVector(A=Add(A=@n60, B=@n70), B=@n62), B=ScalarParameter() {defaultValue=4, group=Cosmos, parameterName=Stars BG Tile}) {constB=6}
  n72 = NamedRerouteUsage()
  n73 = Multiply(
   A=Saturate(None=Add(A=@n69) {constB=-0.05})
   B=Multiply(
    A=TextureSample(UVs=@n71, Tex=@n72).R
    B=VectorParameter() {group=Cosmos, parameterName=Stars BG Color, defaultValue=<rgba 27.2572 40.2784 115 1>}.RGB
   )
  )
  n74 = Saturate(None=Add(A=NamedRerouteUsage(), B=Constant() {r=0.3}))
  n75 = Add(A=Multiply(A=Saturate(None=@n49), B=Multiply(A=@n55.R, B=@n56.RGB)), B=Multiply(A=Add(A=@n68, B=@n73), B=@n74))
  n76 = Add(A=@n47, B=Constant() {r=0.6}) {constB=0.6, desc=0.6}
  n77 = Add(A=@n51, B=Multiply(A=@n52) {constB=0.5})
  MaterialExpressionReroute_4 = Reroute(None=TextureSample(UVs=AppendVector(A=@n77, B=@n54), Tex=NamedRerouteUsage()).B)
  n78 = Multiply(
   A=@MaterialExpressionReroute_4
   B=Multiply(A=VectorParameter() {group=Cosmos, parameterName=Clouds Color, defaultValue=<rgba 0.5391 0.7589 1.5 1>}.RGB, B=@n74)
  )
  n79 = LinearInterpolate(A=Add(A=Multiply(A=@n48, B=@n75), B=Multiply(A=@n76, B=@n78)), B=@n35.RGB, Alpha=Add(A=Multiply(A=@n41, B=@n42), B=Add(A=@n45, B=@n46)))
  n80 = TextureCoordinate()
  n81 = TextureSample(UVs=@n80, Tex=NamedRerouteUsage())
  n82 = Multiply(A=NamedRerouteUsage()) {constB=2}
  n83 = Multiply(
   A=Saturate(None=Add(A=@n82) {constB=0.15})
   B=VectorParameter() {group=Cosmos, parameterName=BlueLines Color, defaultValue=<rgba 0.1192 0.1827 0.5 1>}.RGB
  )
  MaterialExpressionReroute_5 = Reroute(None=Add(A=Multiply(A=@n34, B=@n79) {constA=1}, B=Multiply(A=@n81.A, B=@n83) {constB=10}))
  n84 = SetMaterialAttributes!(
   MaterialAttributes
   Base Color=LinearInterpolate(A=@n30, B=Multiply(A=@n18, B=@n30) {constA=5}, Alpha=NamedRerouteDeclaration(None=@n31) {name=Lightdir, nodeColor=<rgba 1 0.0938 0 1>})
   Emissive Color=@MaterialExpressionReroute_5
   Specular=@n12
   Roughness=Add(A=Multiply(A=@n11.B, B=@n13), B=@n14)
   Metallic=Add(A=Multiply(A=@n11.G, B=@n15), B=ScalarParameter() {defaultValue=0.3, desc=0.7, group=Cosmos, parameterName=Metallic Add, sortPriority=0})
  ) {attributeSetTypes [
    #1
    #2
    #3
    #4
    #5
   ]}
  Composite!(MaterialAttributes=@n84) {subgraphName=Documentation Link}
  Composite(MaterialAttributes=@MaterialExpressionReroute_5.Output) {desc=Game Shared WPO and Fire Effect, subgraphName=WPO And Shared Functions}
  Constant3Vector() {constant=<rgba 0 0 1 0>}
  Constant() {r=0.2}
  Constant() {r=2.5}
  LinearInterpolate(Alpha=@n12)
  n85 = Multiply(A=NamedRerouteUsage(), B=Constant() {r=0.5}) {desc=1.3}
  n86 = Constant() {r=-0.3}
  LinearInterpolate(
   A=@n14
   B=ScalarParameter() {defaultValue=1, desc=0.6, group=Cosmos, parameterName=Roughness_Add Center, sortPriority=0}
   Alpha=Saturate(None=Add(A=@n85, B=@n86) {desc=-0.8})
  )
  n87 = ObjectPositionWS()
  MaterialFunctionCall!(Distance Field 1=@n87, Distance Field 2) {materialFunction=/Engine/Functions/DistanceFields/Combiners/DistanceField_Union}
  n88 = MaterialFunctionCall(Material_In=@MaterialExpressionReroute_4.MaterialAttributes) {materialFunction=&MF_FN_DestructibleDamage}
  MaterialFunctionCall!(Material Attributes=MaterialFunctionCall!(In=@n88.Material_Out) {materialFunction=/Game/Athena/WinterArt/Materials/MF_WPO_WindNew}.Result) {materialFunction=/Game/Creative/Sets/MilitaryBase/Props/MF_FanSpin}
  n89 = MaterialFunctionCall!(In=@n88.Material_Out, UseSphericalMappingOverride) {materialFunction=/Game/Packages/Fortress_SharedMaterials/Base_Material_Functions/MF_CurieFallbackEffect}
  MaterialFunctionCall!(In=@n89.Result, Custom Mask, Custom Rotation Axis) {materialFunction=/Game/Athena/WinterArt/Materials/MF_WindWPO_v2}
  n90 = Multiply(A=NamedRerouteUsage()) {constB=-0.01}
  Multiply(A=@n90, B=Constant() {r=-0.5})
  n91 = Divide(A=Distance(A=ObjectPositionWS(), B=CameraPositionWS()), B=Constant() {r=8192}) {constB=512}
  n92 = OneMinus(None=Clamp(None=@n91, Max=Constant() {r=1}))
  Multiply(A=@n35.RGB, B=NamedRerouteDeclaration(None=@n92) {name=DistanceFade, nodeColor=<rgba 1 0 0.8437 1>})
  NamedRerouteDeclaration(None=TextureObject() {texture=/Game/Environments/Helios/Props/Sungold/Sungold_AtlasStatue_A/Textures/Nite/T_NITE_Sungold_AtlasStatue_Colors}) {name=T_ColorVariation, nodeColor=<rgba 0 1 0.3906 1>}
  NamedRerouteDeclaration(None=TextureObject() {texture=/Game/Environments/Helios/Props/Sungold/Sungold_AtlasStatue_A/Textures/Nite/T_NITE_Sungold_AtlasStatue_Masks_01}) {name=T_Masks, nodeColor=<rgba 1 0.0703 0 1>}
  n93 = TextureObject() {
   samplerType=SAMPLERTYPE_Normal
   texture=/Game/Environments/Helios/Props/Sungold/Sungold_AtlasStatue_A/Textures/Nite/T_NITE_Sungold_AtlasStatue_A_B_N
  }
  NamedRerouteDeclaration(None=@n93) {name=T_N, nodeColor=<rgba 0 0.25 1 1>}
  n94 = MaterialFunctionCall!(
   Sphere Position
   Sphere Radius=Multiply(A=ObjectBounds(), B=Constant() {desc=fix for switch, r=1.03})
   RayDirection=MaterialFunctionCall!(Use World Position Offsets) {materialFunction=/Engine/Functions/Engine_MaterialFunctions02/Utility/CameraVectorWithWPOOptions}.Result
   World position=Subtract(A=WorldPosition().XYZ, B=ObjectPositionWS())
  ) {materialFunction=/Engine/Functions/Engine_MaterialFunctions02/RayTracedSphere}
  n95 = DotProduct(A=Normalize(VectorInput=@n94.Intersection Surface Normal), B=CameraVectorWS())
  NamedRerouteDeclaration(None=OneMinus(None=NamedRerouteDeclaration(None=@n95) {name=Fresnel, nodeColor=<rgba 0.7578 1 0 1>})) {name=1-Fresnel, nodeColor=<rgba 1 0 0.8203 1>}
  n96 = MaterialFunctionCall() {materialFunction=&ObjectPivotPoint}
  n97 = Constant3Vector() {constant=<rgba 0 -15 530 1>}
  n98 = Multiply(A=MaterialFunctionCall(Position=Add(A=@n96.Object Pivot Location, B=@n97)) {materialFunction=&ObjectStabilizedScreenSpaceUVs}.StabilizedUVs)
  n99 = Normalize(VectorInput=Transform(None=@n94.Intersection Surface Normal) {transformSourceType=TRANSFORMSOURCE_World, transformType=TRANSFORM_View})
  n100 = Constant3Vector() {constant=<rgba 0.5 -0.5 0 0>}
  n101 = ComponentMask(None=Add(A=Multiply(A=@n99, B=@n100) {constB=-0.5}) {constB=0.5}) {g=true, r=true}
  NamedRerouteDeclaration(None=Add(A=@n98, B=Subtract(A=@n98, B=@n101))) {name=UV_Matcap_BG, nodeColor=<rgba 0.474 0 0.2333 1>}
  n102 = Multiply(A=@n6, B=Constant() {r=3}) {constB=3, desc=3}
  n103 = NamedRerouteUsage()
  n104 = Multiply(A=@n25, B=Constant() {r=8}) {constB=8, desc=8}
  n105 = NamedRerouteUsage()
  n106 = Multiply(A=Multiply(A=TextureSample(UVs=@n102, Tex=@n103).RGBA, B=TextureSample(UVs=@n104, Tex=@n105).RGBA)) {constB=2}
  NamedRerouteDeclaration(None=ComponentMask(None=@n106) {g=true}) {name=StarShimmeringMask, nodeColor=<rgba 1 0.3047 0 1>}
  NamedRerouteDeclaration(None=ComponentMask(None=@n106) {r=true}) {name=StarShimmeringMask_1, nodeColor=<rgba 1 0.3047 0 1>}
  NamedRerouteDeclaration(None=@n92) {name=Debug, nodeColor=<rgba 0.9688 1 0 1>}
  NamedRerouteDeclaration(None=TextureObject() {texture=/Game/Environments/Helios/Props/Sungold/Sungold_AtlasStatue_A/Textures/Nite/T_NITE_Sungold_AtlasStatue_ConstelationsPolarCoord}) {name=T_Constelations_Polar, nodeColor=<rgba 1 0 0.4922 1>}
  n107 = NamedRerouteUsage()
  NamedRerouteDeclaration(None=Rotator(Coordinate=@n80, Time=Multiply(A=@n107) {constB=0.05})) {name=Rotating_UV, nodeColor=<rgba 0 1 0.7188 1>}
  NamedRerouteDeclaration(None=OneMinus(None=@n8)) {name=1-Fresnel_1, nodeColor=<rgba 1 0 0.8203 1>}
  NamedRerouteDeclaration(None=TextureObject() {texture=/Game/Environments/Helios/Props/Sungold/Sungold_AtlasStatue_A/Textures/Nite/T_NITE_Sungold_AtlasStatue_Constelations}) {name=T_Constelations, nodeColor=<rgba 0 0.4141 1 1>}
  NamedRerouteDeclaration(None=TextureObject() {samplerType=SAMPLERTYPE_LinearColor, texture=&T_NoisePack_LinearColor}) {name=T_ConstelationWobble, nodeColor=<rgba 0.7344 1 0 1>}
  NamedRerouteDeclaration() {name=T_S, nodeColor=<rgba 1 0 0.0469 1>}
  NamedRerouteDeclaration() {name=T_M, nodeColor=<rgba 1 0 0.0469 1>}
  NamedRerouteDeclaration() {name=T_R, nodeColor=<rgba 0.5234 0 1 1>}
  n108 = TextureObject() {
   samplerType=SAMPLERTYPE_Masks
   texture=/Game/Environments/Helios/Props/Sungold/Sungold_AtlasStatue_A/Textures/Nite/T_NITE_Sungold_AtlasStatue_A_B_S
  }
  NamedRerouteDeclaration(None=@n108) {name=T_SMR, nodeColor=<rgba 0.3359 1 0 1>}
  n109 = MaterialFunctionCall() {materialFunction=&LocalPosition}
  n110 = Multiply(A=ComponentMask(None=VertexInterpolator(VS=@n109."Instance Local Position (Excluding Offsets)").PS) {b=true}) {constB=0.001}
  NamedRerouteDeclaration(None=Saturate(None=Add(A=@n110) {constB=0.1})) {name=Mask_Vertical Gradient, nodeColor=<rgba 0.6172 0 1 1>}
  NamedRerouteDeclaration(None=TextureObject() {texture=/Game/Environments/Helios/Props/Sungold/Sungold_AtlasStatue_A/Textures/Nite/T_NITE_Sungold_AtlasStatue_A_B_D}) {name=T_BaseColor, nodeColor=<rgba 0 0.9766 1 1>}
  NamedRerouteUsage()
  NamedRerouteUsage()
  NamedRerouteUsage()
  NamedRerouteUsage()
  PinBase() {reroutePins [MaterialAttributes {expression=&self:MaterialGraph_0.MaterialGraphNode_Composite_0.Documentation Link.MaterialExpressionReroute_4}]}
  n111 = ScalarParameter() {
   desc=Documentation link
   group="0 (Expand)  Documentation (RMB - Copy display name) and Debug view"
   parameterName=https://confluence.it.epicgames.com/display/FORT/Fortnite+Base+Material+2022
  }
  PinBase!(Output=SetMaterialAttributes!(MaterialAttributes=@MaterialExpressionReroute_4.MaterialAttributes, World Position Offset=Multiply(A=@n111)) {attributeSetTypes [#6]}) {reroutePins [Output {expression=&self:MaterialGraph_0.MaterialGraphNode_Composite_0.Documentation Link.MaterialExpressionReroute_5}]}
  PinBase!() {reroutePins [MaterialAttributes {expression=&self:MaterialGraph_0.MaterialGraphNode_Composite_1.WPO And Shared Functions.MaterialExpressionReroute_4}]}
  PinBase!(Result=@n89.Result) {reroutePins [Result {expression=&self:MaterialGraph_0.MaterialGraphNode_Composite_1.WPO And Shared Functions.MaterialExpressionReroute_5}]}
  PixelNormalWS()
  Reroute(None=AppendVector(A=Add(A=@n51, B=@n90), B=@n54))
  n112 = Add(A=Multiply(A=NamedRerouteUsage(), B=Constant() {r=10}), B=Constant() {r=-0.3}) {constB=0.5}
  n113 = Multiply(A=@n63, B=Constant() {r=1}) {constB=2.5}
  n114 = NamedRerouteUsage()
  Reroute(None=Multiply(A=Saturate(None=@n112), B=TextureSample(UVs=@n113, Tex=@n114).G))
  Reroute(None=Subtract(A=WorldPosition().XYZ, B=ObjectPositionWS()))
  Rotator(Coordinate=TextureCoordinate(), Time=Multiply(A=NamedRerouteUsage()) {constB=0.1})
  Saturate(None=@n91)
  Saturate(None=Distance(A=@n87, B=Constant() {r=0.1}))
  SetMaterialAttributes!(MaterialAttributes, Base Color, Emissive Color=NamedRerouteUsage(), Specular, Roughness, Metallic) {attributeSetTypes [
    #1
    #2
    #3
    #4
    #5
   ]}
  Step(Y=ComponentMask(None=NamedRerouteDeclaration(None=@n101) {name=UV_Matcap_FG, nodeColor=<rgba 1 0 0.4922 1>}) {r=true}) {constX=0.5}
  TextureObject() {texture=/Game/Environments/Helios/Props/Sungold/Sungold_AtlasStatue_A/Textures/Nite/T_Nite_Sungold_AtlasStatue_Masks_Packing}
  TextureObject() {
   samplerType=SAMPLERTYPE_LinearGrayscale
   texture=/Game/Environments/Helios/Props/Sungold/Sungold_AtlasStatue_A/Textures/Nite/T_NITE_Sungold_AtlasStatue_Masks_02
  }
  VertexNormalWS()
 ]
}
